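; Coastal vendor cart prop building actor that reports a quest stat when something lands on its umbrella.
; Overlap of the Jump_Collision component resolves the Fort player controller from the overlapping actor and gets its Athena quest manager.
; It then sends a complex custom stat event tagged Athena.Quests.Jump.OnUmbrella with a count of one.
; The construction script is empty.
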(event OnComponentBeginOverlap(Jump_Collision)
  (params (OverlappedComponent :type Primitive%20Component%20Object%20Reference) (OtherActor :type Actor%20Object%20Reference) (OtherComp :type Primitive%20Component%20Object%20Reference) (OtherBodyIndex :type Integer) (bFromSweep :type Boolean) (SweepResult :type Hit%20Result%20Structure%20%28by%20ref%29))
  (Quest|SendComplexCustomStatEvent (Quests|GetQuestManager (Fortnite|GetFortPlayerControllerfromActor OtherActor) "Athena") 0 "(GameplayTags=((TagName=\"Athena.Quests.Jump.OnUmbrella\")))" "(GameplayTags=)" 1))

(fn ConstructionScript)
