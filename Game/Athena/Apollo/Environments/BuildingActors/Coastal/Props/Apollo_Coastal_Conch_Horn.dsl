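; Coastal conch horn prop building actor that a player can interact with under a "Toot Conch" prompt to play a sound.
; On interaction it stores its own location as SpawnLocation, then calls PlayHorn.
; PlayHorn spawns the SoundCue there at the Pitch variable, only on the Remote (non-authority) branch of an authority switch.
; CanThePlayerInteract gates interaction, and the icon sits slightly above the actor.
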(event Interaction|EventBlueprintOnInteract
  (params (InteractingPawn :type Fort%20Pawn%20Object%20Reference) (InteractionBeingAttempted :type EInteractionBeingAttempted%20Enum :default "FirstInteraction"))
  (Variables|Default|SetSpawnLocation (Transformation|GetActorLocation self))
  (CallFunction|PlayHorn))

(event Custom|PlayHorn
  (Networking|SwitchHasAuthority
    (:Remote
      (Audio|SpawnSoundatLocation :Sound (Variables|Default|GetSoundCue) :Location (Variables|Default|GetSpawnLocation) :VolumeMultiplier 1.0 :PitchMultiplier (Variables|Default|GetPitch) :StartTime 0.0 :bAutoDestroy true))))

(fn BlueprintCanInteract
  (params (InteractingPawn :type Fort%20Pawn%20Object%20Reference) (InteractionBeingAttempted :type EInteractionBeingAttempted%20Enum :default "FirstInteraction") (InteractionType :type TInteractionType%20Enum :default "IT_NoInteraction"))
  (return (Variables|Default|GetCanThePlayerInteract)))

(fn BlueprintGetInteractionString
  (params (InteractingPawn :type Fort%20Pawn%20Object%20Reference) (InteractionBeingAttempted :type EInteractionBeingAttempted%20Enum :default "FirstInteraction"))
  (return "NSLOCTEXT(\"[7BB6720D43EC7D64E4BA55B3D2E3944B]\", \"D198C4154B285F920D7BFCABC61D8216\", \"Toot Conch\")"))

(fn GetIconPlacement
  (params (SelfActor :type Actor%20Object%20Reference) (ViewingActor :type Actor%20Object%20Reference))
  (return (+ (Transformation|GetActorLocation SelfActor) "0, 0,50.000000") "0, 0, 0"))

(fn ConstructionScript)
