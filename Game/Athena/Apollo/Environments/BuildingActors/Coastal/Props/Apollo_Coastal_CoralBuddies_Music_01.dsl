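; Coral Buddies music prop for the Apollo coastal environment that registers its ambient sound as a diegetic point-of-interest music event.
; On BeginPlay it runs the parent logic, then does a validated get on its Buddy Ambient Sound component.
; If valid, it adds that component to the MusicEventSubsystem under the diegetic POI tag.
; The construction script is empty.
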(event EventBeginPlay
  (|CallParentFunction)
  (bind _buddy_ambient_sound (Variables|Default|GetBuddy_Ambient_Sound)
    (:then
      (Audio|AddEventbyComponent (GameInstanceSubsystems|GetMusicEventSubsystem) _buddy_ambient_sound "(TagName=\"MusicEventSubsystem.Event.POI.Diegetic\")" "(GameplayTags=)" 0 ""))
    (:else)))

(fn ConstructionScript)
