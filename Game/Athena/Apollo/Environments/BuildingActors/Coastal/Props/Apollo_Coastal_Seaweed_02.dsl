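; Coastal seaweed prop building actor that removes itself from view on lower-end platforms.
; On BeginPlay it runs the parent logic, then, only where cosmetic events can execute, sets its hidden-in-game state to true on Nintendo Switch or mobile and false elsewhere.
; The construction script is empty, so this platform visibility check is the only logic the file adds.
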(event EventBeginPlay
  (|CallParentFunction)
  (Networking|CanExecuteCosmeticEvents
    (:True
      (Rendering|SetActorHiddenInGame (or (TouchPlatform|IsNintendoSwitch) (TouchPlatform|IsMobileGame))))))

(fn ConstructionScript)
